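# Anim Blueprint for the Troops book back bling, which opens and flips pages when its wearer eliminates enemies.
# Elimination events, or timed increments in frontend preview, raise KillCount, opening the book once landing has unlocked it.
# The Pack state machine flips up to four pages by count, with a randomly chosen looping page-turn style after higher counts.
# Curves drive float, cover and rotation additives; a cosmetic variant check selects reactive or closed pose.
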AnimBlueprint /BRCosmetics/Accessories/FORT_Backpacks/Backpack_M_MED_Troops/Meshes/M_MED_Troops_Pack_AnimBP
target_skeleton=&/M_MED_Troops_Pack_Skeleton
graphs [
 AnimGraph {nodes [
   n1 = "Transform (Modify) Bone - Bone: JNT_Scholar_Main"(ComponentPose=Local To Component(LocalPose=Additive Identity Pose()), Alpha=1, Translation="0,0,0", Rotation="0,-90,0", Scale="1,1,1") {
    rotationMode=BMM_Additive
    rotationSpace=BCS_BoneSpace
    alphaBoolBlend {blendOption=Linear}
    boneToModify {boneName=JNT_Scholar_Main}
    rotation {yaw=-90}
   }
   Apply Additive!(Base, Additive=Component To Local(ComponentPose=@n1), AlphaCurveName:Name=AdditionalRotation) -> (Pose) {alphaInputType=Curve, alphaBoolBlend {blendOption=Linear}}
   n2 = Apply Additive(
    Base=Sequence Player 'Idle_Closed_PROP'() {sequence=/BRCosmetics/Accessories/FORT_Backpacks/Backpack_M_MED_Troops/Animation/Idle_Closed_PROP}
    Additive="Sequence Player (additive) 'Idle_Open_Float_PROP'"() {sequence=/BRCosmetics/Accessories/FORT_Backpacks/Backpack_M_MED_Troops/Animation/Idle_Open_Float_PROP}
    AlphaCurveName=Floating
   ) {alphaInputType=Curve, alphaBoolBlend {blendOption=Linear}}
   n3 = "Sequence Player (additive) 'Idle_Open_Float_PROP'"() {sequence=&Idle_Open_Float_PROP}
   AnimGraphNode_StateMachine_2 = Pack() {is=title AnimGraphNode_StateMachine}
   AnimGraphNode_StateMachine_1 = PageTurning() {is=title AnimGraphNode_StateMachine}
   n4 = Apply Additive!(
    Base=Apply Additive!(
     Base=Apply Additive(Base=@AnimGraphNode_StateMachine_2, Additive=@n3, AlphaCurveName=Floating) {alphaInputType=Curve, alphaBoolBlend {blendOption=Linear}}
     Additive="Sequence Player (additive) 'Idle_Open_Float_Cover_PROP'"() {sequence=/BRCosmetics/Accessories/FORT_Backpacks/Backpack_M_MED_Troops/Animation/Idle_Open_Float_Cover_PROP}
     AlphaCurveName:Name=FullyOpen
    ) -> (Pose) {alphaInputType=Curve, alphaBoolBlend {blendOption=Linear}}
    Additive=Blend Poses by bool(
     BlendPose_0=@AnimGraphNode_StateMachine_1
     BlendPose_1="Sequence Player (additive) 'Zero_Pose_PROP'"() {sequence=/BRCosmetics/Accessories/FORT_Backpacks/Backpack_M_MED_Troops/Animation/Zero_Pose_PROP}
     bActiveValue=Get UseMultiPageStyle() -> (UseMultiPageStyle:bool)
    ) {blendType=Linear}
    Alpha:float=Get PageTurnAlpha() -> (PageTurnAlpha:double)
   ) -> (Pose) {alphaBoolBlend {blendOption=Linear}}
   n5 = Get CosmeticVariantReactivity() -> (CosmeticVariantReactivity:bool)
   n6 = "Transform (Modify) Bone - Bone: JNT_Scholar_Main"(
    ComponentPose=Local To Component(LocalPose=Blend Poses by bool(BlendPose_0=@n4, BlendPose_1=@n2, bActiveValue=@n5) {blendType=Linear})
    Alpha=Get AdditionalRotationCurveValue() -> (AdditionalRotationCurveValue:float)
    Translation="0,0,0"
    Rotation="0,-45,0"
    Scale="1,1,1"
   ) {
    rotationMode=BMM_Additive
    rotationSpace=BCS_BoneSpace
    alphaBoolBlend {blendOption=Linear}
    boneToModify {boneName=JNT_Scholar_Main}
    rotation {yaw=-45}
   }
   Output Pose(Result=Component To Local(ComponentPose=@n6)) {layerGroup=None}
  ]}
 PageTurning {
  entry=PageTurning
  path=AnimGraph.AnimGraphNode_StateMachine_1.PageTurning
  ids=id
  nodes [AnimStateNode_0 = PageTurning() {is=title AnimStateNode}]
 }
 PageTurning {
  path=AnimGraph.AnimGraphNode_StateMachine_1.PageTurning.AnimStateNode_0.PageTurning
  nodes [Output Animation Pose(Result="Sequence Player (additive) 'Idle_Page_Turning_RightToLeft_PROP'"!(
     Sequence:Anim Sequence Base Object Reference=Get PageTurnAnim() -> (PageTurnAnim:Anim Sequence Base Object Reference)
     PlayRate:float=Get PageFlipSpeed_MutliPages() -> (PageFlipSpeed_MutliPages:double)
    ) -> (Pose))]
 }
 Pack {
  entry=Closed
  path=AnimGraph.AnimGraphNode_StateMachine_2.Pack
  ids=id
  nodes [
   AnimStateNode_0 = Closed() {is=title AnimStateNode}
   AnimStateNode_1 = Open_Into() {is=title AnimStateNode}
   AnimStateNode_2 = Open() {is=title AnimStateNode}
   AnimStateNode_3 = Open_Out() {is=title AnimStateNode}
   AnimStateNode_5 = PageFlip_page4() {is=title AnimStateNode}
   AnimStateNode_6 = PageFlip_page1() {is=title AnimStateNode}
   AnimStateNode_7 = PageFlip_page2() {is=title AnimStateNode}
   AnimStateNode_8 = PageFlip_page3() {is=title AnimStateNode}
   AnimStateTransitionNode_0 = AnimStateTransitionNode()
   AnimStateTransitionNode_1 = AnimStateTransitionNode() {bAutomaticRuleBasedOnSequencePlayerInState=true, crossfadeDuration=0.05}
   AnimStateTransitionNode_10 = AnimStateTransitionNode()
   AnimStateTransitionNode_11 = AnimStateTransitionNode()
   AnimStateTransitionNode_12 = AnimStateTransitionNode()
   AnimStateTransitionNode_2 = AnimStateTransitionNode()
   AnimStateTransitionNode_3 = AnimStateTransitionNode() {bAutomaticRuleBasedOnSequencePlayerInState=true, crossfadeDuration=0.05}
   AnimStateTransitionNode_4 = AnimStateTransitionNode() {crossfadeDuration=0}
   AnimStateTransitionNode_6 = AnimStateTransitionNode() {crossfadeDuration=0}
   AnimStateTransitionNode_7 = AnimStateTransitionNode() {crossfadeDuration=0}
   AnimStateTransitionNode_8 = AnimStateTransitionNode() {crossfadeDuration=0}
   AnimStateTransitionNode_9 = AnimStateTransitionNode()
  ]
  links [
   {
    dst=Open_Into
    rule=AnimGraph.AnimGraphNode_StateMachine_2.Pack.AnimStateTransitionNode_0.Transition
    src=Closed
    via=AnimStateTransitionNode_0
   }
   {dst=Open_Out, rule=AnimGraph.AnimGraphNode_StateMachine_2.Pack.AnimStateTransitionNode_2.Transition, src=Open, via=AnimStateTransitionNode_2}
   {
    dst=PageFlip_page1
    rule=AnimGraph.AnimGraphNode_StateMachine_2.Pack.AnimStateTransitionNode_4.Transition
    src=Open
    via=AnimStateTransitionNode_4
   }
   {
    dst=Open
    rule=AnimGraph.AnimGraphNode_StateMachine_2.Pack.AnimStateTransitionNode_1.Transition
    src=Open_Into
    via=AnimStateTransitionNode_1
   }
   {
    dst=Closed
    rule=AnimGraph.AnimGraphNode_StateMachine_2.Pack.AnimStateTransitionNode_3.Transition
    src=Open_Out
    via=AnimStateTransitionNode_3
   }
   {
    dst=Open_Out
    rule=AnimGraph.AnimGraphNode_StateMachine_2.Pack.AnimStateTransitionNode_9.Transition
    src=PageFlip_page1
    via=AnimStateTransitionNode_9
   }
   {
    dst=PageFlip_page2
    rule=AnimGraph.AnimGraphNode_StateMachine_2.Pack.AnimStateTransitionNode_6.Transition
    src=PageFlip_page1
    via=AnimStateTransitionNode_6
   }
   {
    dst=Open_Out
    rule=AnimGraph.AnimGraphNode_StateMachine_2.Pack.AnimStateTransitionNode_10.Transition
    src=PageFlip_page2
    via=AnimStateTransitionNode_10
   }
   {
    dst=PageFlip_page3
    rule=AnimGraph.AnimGraphNode_StateMachine_2.Pack.AnimStateTransitionNode_7.Transition
    src=PageFlip_page2
    via=AnimStateTransitionNode_7
   }
   {
    dst=Open_Out
    rule=AnimGraph.AnimGraphNode_StateMachine_2.Pack.AnimStateTransitionNode_11.Transition
    src=PageFlip_page3
    via=AnimStateTransitionNode_11
   }
   {
    dst=PageFlip_page4
    rule=AnimGraph.AnimGraphNode_StateMachine_2.Pack.AnimStateTransitionNode_8.Transition
    src=PageFlip_page3
    via=AnimStateTransitionNode_8
   }
   {
    dst=Open_Out
    rule=AnimGraph.AnimGraphNode_StateMachine_2.Pack.AnimStateTransitionNode_12.Transition
    src=PageFlip_page4
    via=AnimStateTransitionNode_12
   }
  ]
 }
 Closed {
  path=AnimGraph.AnimGraphNode_StateMachine_2.Pack.AnimStateNode_0.Closed
  nodes [Output Animation Pose(Result=Sequence Player 'Idle_Closed_PROP'() {sequence=&Idle_Closed_PROP})]
 }
 Open_Into {
  path=AnimGraph.AnimGraphNode_StateMachine_2.Pack.AnimStateNode_1.Open_Into
  nodes [Output Animation Pose(Result=Sequence Player 'Idle_Open_Into_PROP'() {sequence=/BRCosmetics/Accessories/FORT_Backpacks/Backpack_M_MED_Troops/Animation/Idle_Open_Into_PROP})]
 }
 Open {
  path=AnimGraph.AnimGraphNode_StateMachine_2.Pack.AnimStateNode_2.Open
  nodes [Output Animation Pose(Result=Sequence Player 'Idle_Open_PROP'() {sequence=/BRCosmetics/Accessories/FORT_Backpacks/Backpack_M_MED_Troops/Animation/Idle_Open_PROP})]
 }
 Open_Out {
  path=AnimGraph.AnimGraphNode_StateMachine_2.Pack.AnimStateNode_3.Open_Out
  nodes [Output Animation Pose(Result=Sequence Player 'Idle_Open_Out_PROP'() {sequence=/BRCosmetics/Accessories/FORT_Backpacks/Backpack_M_MED_Troops/Animation/Idle_Open_Out_PROP})]
 }
 PageFlip_page4 {
  path=AnimGraph.AnimGraphNode_StateMachine_2.Pack.AnimStateNode_5.PageFlip_page4
  nodes [Output Animation Pose(Result=Sequence Player 'PageFlip_R2L_page4_PROP'!(PlayRate:float=1) -> (Pose) {bLoopAnimation=false, sequence=/BRCosmetics/Accessories/FORT_Backpacks/Backpack_M_MED_Troops/Animation/PageFlip_R2L_page4_PROP})]
 }
 PageFlip_page1 {
  path=AnimGraph.AnimGraphNode_StateMachine_2.Pack.AnimStateNode_6.PageFlip_page1
  nodes [Output Animation Pose(Result=Sequence Player 'PageFlip_R2L_page1_PROP'!(PlayRate:float=1) -> (Pose) {bLoopAnimation=false, sequence=/BRCosmetics/Accessories/FORT_Backpacks/Backpack_M_MED_Troops/Animation/PageFlip_R2L_page1_PROP})]
 }
 PageFlip_page2 {
  path=AnimGraph.AnimGraphNode_StateMachine_2.Pack.AnimStateNode_7.PageFlip_page2
  nodes [Output Animation Pose(Result=Sequence Player 'PageFlip_R2L_page2_PROP'!(PlayRate:float=1) -> (Pose) {bLoopAnimation=false, sequence=/BRCosmetics/Accessories/FORT_Backpacks/Backpack_M_MED_Troops/Animation/PageFlip_R2L_page2_PROP})]
 }
 PageFlip_page3 {
  path=AnimGraph.AnimGraphNode_StateMachine_2.Pack.AnimStateNode_8.PageFlip_page3
  nodes [Output Animation Pose(Result=Sequence Player 'PageFlip_R2L_page3_PROP'!(PlayRate:float=1) -> (Pose) {bLoopAnimation=false, sequence=/BRCosmetics/Accessories/FORT_Backpacks/Backpack_M_MED_Troops/Animation/PageFlip_R2L_page3_PROP})]
 }
 Transition {
  path=AnimGraph.AnimGraphNode_StateMachine_2.Pack.AnimStateTransitionNode_0.Transition
  nodes [Result(bCanEnterTransition=Get IsOpen() -> (IsOpen:bool))]
 }
 Transition {path=AnimGraph.AnimGraphNode_StateMachine_2.Pack.AnimStateTransitionNode_1.Transition, nodes [Result()]}
 Transition {
  path=AnimGraph.AnimGraphNode_StateMachine_2.Pack.AnimStateTransitionNode_10.Transition
  nodes [
   n1 = "Time Remaining (PageFlip_R2L_page2_PROP)"!() -> (ReturnValue:float) {is=title K2Node_AnimGetter}
   n2 = Get KillCount() -> (KillCount:int=0)
   Result(bCanEnterTransition=AND Boolean!(A="Equal (Integer)"(A=@n2, B=2), B="Equal (Float)"(A=@n1), C:bool=true) -> (ReturnValue))
  ]
 }
 Transition {
  path=AnimGraph.AnimGraphNode_StateMachine_2.Pack.AnimStateTransitionNode_11.Transition
  nodes [
   n1 = "Time Remaining (PageFlip_R2L_page3_PROP)"!() -> (ReturnValue:float) {is=title K2Node_AnimGetter}
   n2 = Get KillCount() -> (KillCount=0)
   Result(bCanEnterTransition=AND Boolean(A="Equal (Integer)"(A=@n2, B=3), B="Equal (Float)"(A=@n1), C=true))
  ]
 }
 Transition {
  path=AnimGraph.AnimGraphNode_StateMachine_2.Pack.AnimStateTransitionNode_12.Transition
  nodes [
   n1 = Get KillCount() -> (KillCount=0)
   n2 = "Time Remaining (PageFlip_R2L_page4_PROP)"!() -> (ReturnValue:float) {is=title K2Node_AnimGetter}
   Result(bCanEnterTransition=AND Boolean(A="integer >= integer"(A=@n1, B=4), B="Equal (Float)"(A=@n2), C=true))
  ]
 }
 Transition {
  path=AnimGraph.AnimGraphNode_StateMachine_2.Pack.AnimStateTransitionNode_2.Transition
  nodes [Result(bCanEnterTransition=NOT Boolean(A=Get IsOpen()))]
 }
 Transition {path=AnimGraph.AnimGraphNode_StateMachine_2.Pack.AnimStateTransitionNode_3.Transition, nodes [Result()]}
 Transition {
  path=AnimGraph.AnimGraphNode_StateMachine_2.Pack.AnimStateTransitionNode_4.Transition
  nodes [Result(bCanEnterTransition=Get UsePageCountStyle() -> (UsePageCountStyle:bool))]
 }
 Transition {
  path=AnimGraph.AnimGraphNode_StateMachine_2.Pack.AnimStateTransitionNode_6.Transition
  nodes [
   n1 = Get KillCount() -> (KillCount=0)
   n2 = "Time Remaining (PageFlip_R2L_page1_PROP)"!() -> (ReturnValue:float) {is=title K2Node_AnimGetter}
   Result(bCanEnterTransition=AND Boolean!(A="integer >= integer"(A=@n1, B=2), B="Equal (Float)"(A=@n2)) -> (ReturnValue))
  ]
 }
 Transition {
  path=AnimGraph.AnimGraphNode_StateMachine_2.Pack.AnimStateTransitionNode_7.Transition
  nodes [
   n1 = Get KillCount() -> (KillCount=0)
   n2 = "Time Remaining (PageFlip_R2L_page2_PROP)"()
   Result(bCanEnterTransition=AND Boolean(A="integer >= integer"(A=@n1, B=3), B="Equal (Float)"(A=@n2)))
  ]
 }
 Transition {
  path=AnimGraph.AnimGraphNode_StateMachine_2.Pack.AnimStateTransitionNode_8.Transition
  nodes [
   n1 = Get KillCount() -> (KillCount=0)
   n2 = "Time Remaining (PageFlip_R2L_page3_PROP)"()
   Result(bCanEnterTransition=AND Boolean(A="integer >= integer"(A=@n1, B=4), B="Equal (Float)"(A=@n2)))
  ]
 }
 Transition {
  path=AnimGraph.AnimGraphNode_StateMachine_2.Pack.AnimStateTransitionNode_9.Transition
  nodes [
   n1 = "Time Remaining (PageFlip_R2L_page1_PROP)"()
   n2 = Get KillCount() -> (KillCount=0)
   Result(bCanEnterTransition=AND Boolean!(A="Equal (Integer)"(A=@n2, B=1), B="Equal (Float)"(A=@n1), C:bool=true) -> (ReturnValue))
  ]
 }
 BoolRules {nodes [
   Inputs()
   > n1 = Sequence()
   > Set UseMultiPageStyle(UseMultiPageStyle:bool=AND Boolean!(A=Get FullyOpen() -> (FullyOpen:bool), B=Get MultiPageStyle() -> (MultiPageStyle:bool)) -> (ReturnValue)) -> (Output_Get:bool)
   n2 = Get MultiPageStyle()
   > Set UsePageCountStyle(UsePageCountStyle:bool=AND Boolean!(A=Get Kill() -> (Kill:bool), B=Get FullyOpen(), C:bool=NOT Boolean(A=@n2)) -> (ReturnValue)) -> (Output_Get:bool)
   Outputs!(then:Exec=@n1.then_1)
  ]}
 EventGraph {nodes [
   Event BlueprintUpdateAnimation()
   > Is Valid(InputObject=Get Pawn() -> (Pawn:Fort Player Pawn Object Reference))
   > Set FullyOpen(FullyOpen:bool="Equal (Float)"(A=GetCurveValue(CurveName=FullyOpen), B=1)) -> (Output_Get:bool)
   > Bool Rules!(execute:Exec) -> (then:Exec) {is=title K2Node_MacroInstance}
   > n1 = Sequence!(execute) -> (then_0, then_1, then_2:Exec)
   > Set AdditionalRotationCurveValue(AdditionalRotationCurveValue:float=GetCurveValue(CurveName=AdditionalRotation)) -> (Output_Get:float)
   n2 = Branch(execute=@n1.then_1, Condition=Get bIsSkyDiving())
   > Set IsOpen(IsOpen:bool=false) -> (Output_Get:bool)
   > Set UnlockBook(UnlockBook:bool=false) -> (Output_Get:bool)
   > Set PageTurnAlpha(PageTurnAlpha:double=0) -> (Output_Get:double)
   > Set Timer(Timer:double=0) -> (Output_Get:double)
   > Set Kill(Kill:bool=false) -> (Output_Get:bool)
   Branch(execute=@n2.else, Condition=Get bIsOnGround() -> (bIsOnGround:bool))
   > Set UnlockBook(UnlockBook=true)
   n3 = Branch(execute=@n1.then_2, Condition=Get bShouldDisplayFrontendReactivity() -> (bShouldDisplayFrontendReactivity:bool))
   > Random Method!(execute:Exec) -> (then:Exec) {is=title K2Node_MacroInstance}
   > Kill Counter!(execute:Exec, KillCounts:int=Get KillCount() -> (KillCount=0)) -> (then:Exec) {is=title K2Node_MacroInstance}
   > Kill Method!(execute:Exec) -> (then:Exec) {is=title K2Node_MacroInstance}
   Kill Method(execute=@n3.else)
   Event OnCharacterCustomizationLoaded()
   > n4 = Cast To FortPlayerPawn(Object=TryGetFortPawn())
   > Set Pawn(Pawn:Fort Player Pawn Object Reference=@n4) -> (Output_Get:Fort Player Pawn Object Reference)
   > n5 = Cast To FortPlayerPawn(Object=GetOwningActor())
   > n6 = GetVariantChannelActiveIndexForItemActor!(
    execute
    Item:Athena Cosmetic Item Definition Object Reference=/BRCosmetics/Athena/Items/Cosmetics/Backpacks/Backpack_Troops
    Actor:Actor Object Reference=@n5
    Channel:Gameplay Tag Structure=(TagName="Cosmetics.Variant.Channel.Parts")
    VariantList:Array of Gameplay Tag Structures=Make Array("[0]"=(TagName="Cosmetics.Variant.Property.Stage1"))
    bErrorIfChannelMissing:bool=false
    bErrorIfTagNotInArray:bool=false
   ) -> (then, Index:int=0, bIndexIsValid:bool, bFoundChannel:bool)
   > n7 = Branch(Condition=@n6.bIndexIsValid)
   > Set CosmeticVariantReactivity(CosmeticVariantReactivity:bool=true) -> (Output_Get:bool)
   Set CosmeticVariantReactivity(execute=@n7.else, CosmeticVariantReactivity=false)
   n8 = Event BlueprintProcessAnimInputEvent()
   > Branch(Condition=IsNamed(self=@n8.AnimInputEvent, InEventName=aie.EnemyPlayerEliminated))
   > n9 = Increment Int(Value=Get KillCount() -> (KillCount=0))
   > Kill Counter(KillCounts=@n9)
   AnimNotify_PageTurnOff() {is=title K2Node_Event}
   > Set Kill(Kill=false)
   > Set IsOpen(IsOpen=false)
   > n10 = Branch(Condition="integer >= integer"(A=Get KillCount() -> (KillCount=0), B=3))
   > n11 = Set MultiPageStyle(MultiPageStyle:bool=Select(Option 0:bool=false, Option 1:bool=true, Index:int=RandomIntegerInRange!(Min:int=0, Max:int=1) -> (ReturnValue:int=0)) -> (ReturnValue:bool)) -> (Output_Get:bool)
   Set PageTurnAlpha(execute=[@n11.then, @n10.else], PageTurnAlpha=0)
   > n12 = Branch(Condition="Equal (Integer)"(A=Get LeftRight() -> (LeftRight:int=0)))
   > Set LeftRight(LeftRight:int=1) -> (Output_Get:int=0)
   Set LeftRight(execute=@n12.else, LeftRight=0) -> (Output_Get=0)
   TryGetPawnOwner()
  ]}
 KillCounter {nodes [
   n1 = Inputs!() -> (execute, KillCounts:int)
   > n2 = Sequence!(execute) -> (then_0, then_1)
   > n3 = Set KillCount(KillCount:int=@n1) -> (Output_Get:int=0)
   > Branch(Condition="integer > integer"(A=@n3, B=Get KillCount_Last() -> (KillCount_Last:int=0)))
   > Set Kill(Kill=true)
   > Set KillCount_Last(KillCount_Last:int=Get KillCount() -> (KillCount=0)) -> (Output_Get:int=0)
   n4 = Get KillCount() -> (KillCount=0)
   n5 = SelectInt(A=1, B=@n4, bPickA="Equal (Integer)"(A=@n4))
   > Set PageFlipSpeed(PageFlipSpeed:double=MapRangeClamped(Value="To Float (Integer)"(InInt=@n5), InRangeA=1, InRangeB=6, OutRangeA=1.75, OutRangeB=2.5)) -> (Output_Get:double)
   Outputs(then=@n2.then_1)
  ]}
 KillMethod {nodes [
   Inputs!() -> (execute)
   > n1 = Sequence()
   > Set PageFlipSpeed_MutliPages(PageFlipSpeed_MutliPages:double=Get PageFlipSpeed_New() -> (PageFlipSpeed_New:double)) -> (Output_Get:double)
   n2 = Get bIsSkyDiving()
   > Branch(Condition=AND Boolean!(A=Get UnlockBook() -> (UnlockBook:bool), B=NOT Boolean(A=@n2)) -> (ReturnValue))
   > n3 = Branch(Condition=Get Kill())
   > Set IsOpen(IsOpen=true)
   > Branch(Condition=Get MultiPageStyle())
   > Set PageTurnAnim(PageTurnAnim:Anim Sequence Base Object Reference=Select(
     Option 0:Anim Sequence Base Object Reference=/BRCosmetics/Accessories/FORT_Backpacks/Backpack_M_MED_Troops/Animation/Idle_Page_Turning_RightToLeft_PROP
     Option 1:Anim Sequence Base Object Reference=/BRCosmetics/Accessories/FORT_Backpacks/Backpack_M_MED_Troops/Animation/Idle_Page_Turning_LeftToRight_PROP
     Index=Get LeftRight() -> (LeftRight=0)
    ) -> (ReturnValue:Anim Sequence Base Object Reference)) -> (Output_Get:Anim Sequence Base Object Reference)
   > Set PageTurnAlpha(PageTurnAlpha=1)
   > Delay(Duration=float * float(A=4.2, B=Get PageFlipSpeed_MutliPages()))
   > Set PageFlipSpeed_New(PageFlipSpeed_New:double=RandomFloatInRange(Min=1.15, Max=1.5)) -> (Output_Get:double)
   Set IsOpen(execute=@n3.else, IsOpen=false)
   Outputs(then=@n1.then_1)
  ]}
 NewAnimationLayer {nodes [Output Pose() {layerGroup=None}]}
 RandomMethod {nodes [
   Inputs()
   > Set PageFlipSpeed(PageFlipSpeed=1)
   n1 = Get bIsSkyDiving()
   > Branch(Condition=AND Boolean(A=Get UnlockBook(), B=NOT Boolean(A=@n1)))
   > Delay(Duration=Get Timer() -> (Timer:double))
   > Increment Int(Value=Get KillCount() -> (KillCount=0))
   > Set Timer(Timer=10)
   > Outputs()
  ]}
]
variables [
 AdditionalRotationCurveValue {type=number}
 CosmeticVariantReactivity {type=boolean}
 FullyOpen {type=boolean}
 IsOpen {type=boolean}
 Kill {type=boolean}
 KillCount {type=integer}
 KillCount_Last {type=integer}
 KillCount_PreValue {type=integer}
 LeftRight {type=integer}
 MultiPageStyle {type=boolean}
 PageFlipSpeed {type=number, default=1}
 PageFlipSpeed_MutliPages {type=number, default=1.25}
 PageFlipSpeed_New {type=number, default=1.25}
 PageTurnAlpha {type=number}
 PageTurnAnim {type=AnimSequenceBase, default=&Idle_Page_Turning_RightToLeft_PROP}
 Pawn {type=FortPlayerPawn}
 Timer {type=number, default=2}
 UnlockBook {type=boolean}
 UseMultiPageStyle {type=boolean}
 UsePageCountStyle {type=boolean}
]
